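# Fortnite reticle extension widget for the VanillaSoda weapon, animating material parameters on a centered reticle image for intro, melee, leap and cooldown.
# On construct it plays the intro and listens for the Blade.JumpFX effect, which plays the leap animation and awaits the leap cooldown tag.
# Cooldowns of 4+ seconds play a progress fill time-scaled to their duration.
# A delay then triggers one of two finished animations, chosen by that threshold.
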WidgetBlueprint /VanillaSoda/UI/Widgets/WBP_VanillaSoda_Reticle
animations [
 AnimLeap {
  binding_count=1
  length_seconds=0.1
  object=&self:AnimLeap
  sequence {bindings [Reticle {
     id=#1
     possessed_class=Image
     tracks [{
       type=MovieSceneWidgetMaterialTrack
       sections [{
         type=MovieSceneParameterSection
         end_frame=2
         start_frame=0
         channels [
          CooldownColor {type=MovieSceneScriptingFloatChannel, const [0,1]}
          ProgressFill {type=MovieSceneScriptingFloatChannel, keys [[0,1],[2,0]]}
         ]
        }]
      }]
    }]}
 }
 AnimAttack {
  binding_count=1
  length_seconds=0.4
  object=&self:AnimAttack
  sequence {bindings [Reticle {
     id=#2
     possessed_class=Image
     tracks [{
       type=MovieSceneWidgetMaterialTrack
       sections [{
         type=MovieSceneParameterSection
         end_frame=8
         start_frame=0
         channels [ReticleDistanceX {type=MovieSceneScriptingFloatChannel, keys [[0,0.35],[3.6923,0.4],[8,0.35]]}]
        }]
      }]
    }]}
 }
 AnimFinshedCooldown {
  binding_count=1
  length_seconds=0.4
  object=&self:AnimFinshedCooldown
  sequence {bindings [Reticle {
     id=#3
     possessed_class=Image
     tracks [{
       type=MovieSceneWidgetMaterialTrack
       sections [{
         type=MovieSceneParameterSection
         end_frame=8
         start_frame=0
         channels [
          ReticleDistanceX {type=MovieSceneScriptingFloatChannel, keys [[0,0.35],[3.6923,0.39],[8,0.35]]}
          CooldownColor {type=MovieSceneScriptingFloatChannel, keys [[0,0],[1,1],[6,1],[8,0]]}
          ProgressFill {type=MovieSceneScriptingFloatChannel, const [1,1]}
         ]
        }]
      }]
    }]}
 }
 AnimIntro {
  binding_count=1
  length_seconds=0.4
  object=&self:AnimIntro
  sequence {bindings [Reticle {
     id=#4
     possessed_class=Image
     tracks [
      {
       type=MovieSceneWidgetMaterialTrack
       sections [{
         type=MovieSceneParameterSection
         end_frame=8
         start_frame=0
         channels [ReticleDistanceX {type=MovieSceneScriptingFloatChannel, keys [[3,0.5],[7,0.3]]}]
        }]
      }
      {
       type=MovieSceneFloatTrack
       display_name=Render Opacity
       sections [{type=MovieSceneFloatSection, end_frame=4, start_frame=0, channels [None {type=MovieSceneScriptingFloatChannel, keys [[0,0],[2,1]]}]}]
      }
     ]
    }]}
 }
 AnimCool {
  binding_count=1
  length_seconds=1
  object=&self:AnimCool
  sequence {bindings [Reticle {
     id=#1
     possessed_class=Image
     tracks [{
       type=MovieSceneWidgetMaterialTrack
       sections [{
         type=MovieSceneParameterSection
         end_frame=20
         start_frame=0
         channels [
          CooldownColor {type=MovieSceneScriptingFloatChannel, const [0,2]}
          ProgressFill {type=MovieSceneScriptingFloatChannel, keys [[0,0],[20,1]]}
         ]
        }]
      }]
    }]}
 }
 AnimFinshedCooldownNoProgress {
  binding_count=1
  length_seconds=0.4
  object=&self:AnimFinshedCooldownNoProgress
  sequence {bindings [Reticle {
     id=#3
     possessed_class=Image
     tracks [{
       type=MovieSceneWidgetMaterialTrack
       sections [{
         type=MovieSceneParameterSection
         end_frame=8
         start_frame=0
         channels [
          ReticleDistanceX {type=MovieSceneScriptingFloatChannel, keys [[0,0.35],[3.6923,0.39],[8,0.35]]}
          CooldownColor {type=MovieSceneScriptingFloatChannel, keys [[0,0],[1,1],[6,1],[8,0]]}
          ProgressFill {type=MovieSceneScriptingFloatChannel, keys [[0,0],[1,1]]}
         ]
        }]
      }]
    }]}
 }
]
graphs [
 BindToWeaponEvents {nodes [BindToWeaponEvents!() -> (then, DoBind:bool, EquippedWeapon:Fort Weapon Object Reference) {is=title K2Node_FunctionEntry}]}
 EventGraph {nodes [
   Event Construct()
   > n1 = GetDynamicMaterial(self=Get Reticle() -> (Reticle:Image Object Reference))
   > Set CachedMaterial(CachedMaterial:Material Instance Dynamic Object Reference=@n1) -> (Output_Get:Material Instance Dynamic Object Reference)
   > n2 = WaitGameplayEffectAppliedToActor!(
    execute:Exec
    TargetActor:Actor Object Reference=Get Controlled Pawn!(self:Controller Object Reference=GetOwningPlayer()) -> (ReturnValue:Pawn Object Reference) {is=title K2Node_CallFunction}
    SourceFilter:Gameplay Target Data Filter Handle Structure
    SourceTagRequirements:Gameplay Tag Requirements Structure
    TargetTagRequirements:Gameplay Tag Requirements Structure
    AssetTagRequirements:Gameplay Tag Requirements Structure=Make Gameplay Tag Requirements!(
     RequireTags:Gameplay Tag Container Structure=(GameplayTags=((TagName="GameplayEffect.Blade.JumpFX")))
     IgnoreTags:Gameplay Tag Container Structure=(GameplayTags=)
     TagQuery:Gameplay Tag Query Structure=(TokenStreamVersion=0,TagDictionary=,QueryTokenStream=,UserDescription="",AutoDescription="")
    ) -> (GameplayTagRequirements:Gameplay Tag Requirements Structure) {is=title K2Node_MakeStruct}
    GrantedTagRequirements:Gameplay Tag Requirements Structure
    TriggerOnce:bool=false
    ListenForPeriodicEffect:bool=false
   ) -> (
    then:Exec
    AsyncTaskProxy:Ability Async Wait Gameplay Effect Applied Object Reference
    OnApplied:Exec
    Source:Actor Object Reference
    SpecHandle:Gameplay Effect Spec Handle Structure
    ActiveHandle:Active Gameplay Effect Handle Structure
   ) {is=title K2Node_AsyncAction}
   > Set Async Action(Async Action:Ability Async Wait Gameplay Effect Applied Object Reference=@n2) -> (Output_Get:Ability Async Wait Gameplay Effect Applied Object Reference)
   n3 = Get AnimIntro() -> (AnimIntro:Widget Animation Object Reference)
   > StopAnimation(InAnimation=@n3)
   > PlayAnimation(InAnimation=@n3)
   OnSlamStart(execute=@n2.OnApplied)
   > n4 = WaitGameplayTagAddToActor!(
    execute:Exec
    TargetActor:Actor Object Reference=Get PlayerState!(self:Controller Object Reference=GetOwningPlayer()) -> (PlayerState:Player State Object Reference)
    Tag:Gameplay Tag Structure=(TagName="Granted.Cooldown.VanillaSoda.Leap")
    OnlyTriggerOnce:bool=true
   ) -> (then:Exec, AsyncTaskProxy:Ability Async Wait Gameplay Tag Added Object Reference, Added:Exec) {is=title K2Node_AsyncAction}
   n5 = Sequence(execute=@n4.Added)
   n6 = GetCooldownTimeRemaining!(self:Fort Weapon Reticle Extension Widget Base Object Reference) -> (RemainingTime:float, TotalDuration:float)
   > Delay(Duration=@n6.TotalDuration)
   n7 = "float >= float"(A=@n6.TotalDuration, B=4)
   > PlayAnimation(InAnimation=Select(
     Option 0:Widget Animation Object Reference=Get AnimFinshedCooldownNoProgress() -> (AnimFinshedCooldownNoProgress:Widget Animation Object Reference)
     Option 1:Widget Animation Object Reference=Get AnimFinshedCooldown() -> (AnimFinshedCooldown:Widget Animation Object Reference)
     Index=@n7
    ) -> (ReturnValue:Widget Animation Object Reference))
   Branch(execute=@n5.then_1, Condition=@n7)
   > OnSlamEnd!(execute, self, CooldownTime:float=@n6.TotalDuration) -> (then)
   Event Destruct()
   > Set CachedMaterial()
   > EndAction(self:Ability Async Object Reference=Get Async Action() -> (Async Action:Ability Async Wait Gameplay Effect Applied Object Reference))
   > Set Async Action()
  ]}
 OnMelee {nodes [
   OnMelee() {is=title K2Node_FunctionEntry}
   > PlayAnimation(InAnimation=Get AnimAttack() -> (AnimAttack:Widget Animation Object Reference))
  ]}
 OnSlamEnd {nodes [
   n1 = OnSlamEnd!() -> (then, CooldownTime:float=~) {is=title K2Node_FunctionEntry}
   > PlayAnimation(InAnimation=Get AnimCool() -> (AnimCool:Widget Animation Object Reference), PlaybackSpeed=float / float(A=1, B=@n1))
  ]}
 OnSlamStart {nodes [
   OnSlamStart() {is=title K2Node_FunctionEntry}
   > PlayAnimation(InAnimation=Get AnimLeap() -> (AnimLeap:Widget Animation Object Reference))
  ]}
]
variables [
 Async Action {type=AbilityAsync_WaitGameplayEffectApplied, default=null}
 CachedMaterial {type=MaterialInstanceDynamic}
]
widgets {
 info {
  inheritedWidgetCount=1
  parentClass=/Script/FortniteUI.FortWeaponReticleExtensionWidgetBase
  rootWidgetClass=&Overlay
  widgetCount=3
 }
 widgets [
  Overlay_24 {widgetClassPath=&Overlay, widget=Overlay_24}
  Reticle {
   widgetClassPath=&Image
   bIsVariable=true
   parent=Overlay_24
   widget=Reticle
   properties {brush {
     bIsDynamicallyLoaded=false
     drawAs=Image
     imageType=FullColor
     mirroring=NoMirror
     resourceName=None
     resourceObject=/VanillaSoda/UI/Art/Materials/M_UI_VanillaSoda_Reticle
     tiling=NoTile
     imageSize=<172 172>
     margin=<ltrb 0 0 0 0>
     outlineSettings {
      bUseBrushTransparency=false
      roundingType=HalfHeightRadius
      width=0
      color {colorUseRule=UseColor_Specified, specifiedColor=<rgba 0 0 0 0>}
      cornerRadii=<0 0 0 0>
     }
     tintColor {colorUseRule=UseColor_Specified, specifiedColor=<rgba 1 1 1 1>}
     uVRegion {bIsValid=false, max=<0 0>, min=<0 0>}
    }}
   slot {class=OverlaySlot, properties {horizontalAlignment=HAlign_Center, verticalAlignment=VAlign_Center}}
  }
  TutorialHighlight {widgetClassPath=&TutorialHighlight}
 ]
}
